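# Animation Blueprint for the face accessory part of the male medium PaperBlaze Mask character skin, which follows poses copied from other meshes.
# The base pose copies from the attached parent, with a copy from the owner's skeletal mesh, cached as BodyMesh at initialization, layered over both upper-arm branches.
# The blended pose feeds the Control Rig of the original PaperBlaze body, reused here as the final step before output.
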AnimBlueprint /BRCosmetics/Characters/Player/Male/Medium/Bodies/M_MED_PaperBlaze_Mask/Meshes/Parts/M_MED_PaperBlaze_Mask_FaceAcc_AnimBP
target_skeleton=&/M_MED_PaperBlaze_Mask_FaceAcc_Skeleton
graphs [
 AnimGraph {nodes [
   n1 = Get BodyMesh() -> (BodyMesh:Skeletal Mesh Component Object Reference)
   n2 = Control Rig(Source=Layered blend per bone(
     BasePose=Copy Pose From Mesh() {bUseAttachedParent=true}
     BlendPoses_0=Copy Pose From Mesh(SourceMeshComponent=@n1) {bUseAttachedParent=true}
    ) {layerSetup [{branchFilters [{blendDepth=1, boneName=upperarm_l}, {blendDepth=1, boneName=upperarm_r}]}]}) {
    alphaBoolBlend {blendOption=Linear}
    controlRigAssetReference {blueprintRigClass=/BRCosmetics/Characters/Player/Male/Medium/Bodies/M_MED_PaperBlaze/Meshes/M_MED_PaperBlaze_CtrlRig.M_MED_PaperBlaze_CtrlRig_C}
   }
   Output Pose(Result=@n2) {layerGroup=None}
  ]}
 EventGraph {nodes [
   Event BlueprintInitializeAnimation()
   > Set BodyMesh(BodyMesh=GetOwnerSkeletalMeshComponent())
   Event BlueprintUpdateAnimation()
  ]}
]
variables [BodyMesh {type=SkeletalMeshComponent}]
